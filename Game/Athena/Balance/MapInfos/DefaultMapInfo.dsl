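; Default Athena map info Blueprint that can shrink and speed up storm circles when few players remain.
; TryDynamicStormCircleModification skips creative playlists and acts only when living players are below the current phase's PlayerCap.
; It then reads minimum wait time, shrink time and radius from AthenaGameData curve rows keyed by phase index.
; Missing values are meant to leave the circle unmodified; the event graph is empty.
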(event EventBeginPlay)

(event Collision|EventActorBeginOverlap
  (params (OtherActor :type Actor%20Object%20Reference)))

(event EventTick
  (params (DeltaSeconds :type Float%20%28single-precision%29)))

(fn TryDynamicStormCircleModification
  (params (RoundData_CurrentPlaylist :type Fort%20Playlist%20Athena%20Soft%20Object%20Reference) (RoundData_PhaseIndex :type Integer) (RoundData_LivingPlayers :type Integer) (NewSafeZoneWaitTime :type Float%20%28single-precision%29) (NewSafeZoneShrinkTime :type Float%20%28single-precision%29) (NewSafeZoneRadius :type Float%20%28single-precision%29))
  (if (Utilities|IsValidSoftObjectReference RoundData_CurrentPlaylist)
    ;[1 1 1] Skip modification in creative
    (bind _returnvalue (Class|FortPlaylistAthenaHelper|IsAnyCreativePlaylist (Utilities|ResolveSoftReference RoundData_CurrentPlaylist)))
    (if (not _returnvalue)
      (LocalVariables|SetPhaseIndexFloat (Math|Conversions|ToFloat(Integer) RoundData_PhaseIndex))
      (LocalVariables|SetLivingPlayersFloat (Math|Conversions|ToFloat(Integer) RoundData_LivingPlayers))
      ;[1 1 1] Only modify the circle if the player count is BELOW the PlayerCap threshold
      (bind _outxy (DataTable|EvaluateCurveTableRow "/Game/Athena/Balance/DataTables/AthenaGameData.AthenaGameData" "Default.SafeZone.DynamicStorm.PlayerCap" (LocalVariables|GetPhaseIndexFloat))
        (:RowFound
          (Math|Float|CompareFloat (LocalVariables|GetLivingPlayersFloat) _outxy
            (:>
              _)
            (:==
              _)
            (:<
              ;[1 1 1] Return the values for this phase; if any value is not found it will default to NOT modifying this circle
              (bind _outxy_2 (DataTable|EvaluateCurveTableRow "/Game/Athena/Balance/DataTables/AthenaGameData.AthenaGameData" "Default.SafeZone.DynamicStorm.MinRadius" (LocalVariables|GetPhaseIndexFloat))
                (:RowFound
                  (bind _outxy_3 (DataTable|EvaluateCurveTableRow "/Game/Athena/Balance/DataTables/AthenaGameData.AthenaGameData" "Default.SafeZone.DynamicStorm.MinShrinkTime" (LocalVariables|GetPhaseIndexFloat))
                    (:RowFound
                      (bind _outxy_4 (DataTable|EvaluateCurveTableRow "/Game/Athena/Balance/DataTables/AthenaGameData.AthenaGameData" "Default.SafeZone.DynamicStorm.MinWaitTime" (LocalVariables|GetPhaseIndexFloat))
                        (:RowFound
                          (return _outxy_4 _outxy_3 _outxy_2))
                        (:RowNotFound
                          _)))
                    (:RowNotFound
                      _))
                  (return -1.0 -1.0 -1.0))
                (:RowNotFound
                  _)))))
        (:RowNotFound
          _)))
    (else
      _)))

(fn ConstructionScript)
